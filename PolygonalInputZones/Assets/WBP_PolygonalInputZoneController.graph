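# Widget-less PolygonalInputZoneControllerBase subclass that adds debug drawing for polygonal touch input zones and swipes.
# OnPaint, skipped in distribution builds, draws zone polygons in reverse order, scaling vertices by widget size, with cycling colors or gray when inactive.
# A console variable or local flag enables zone drawing.
# OnSwipeSuccess stores swipe endpoints for one second, drawn as an offset red line when the swipe debug console variable is on.
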WidgetBlueprint /PolygonalInputZones/Assets/WBP_PolygonalInputZoneController
graphs [
 EventGraph {nodes [
   n1 = Event OnSwipeSuccess!() -> (OutputDelegate, then, TouchIndex:int=0, Distance:float, Time:double, StartPos:Vector 2D Structure, EndPos:Vector 2D Structure) {is=title K2Node_Event}
   > Set bShowDebugSwipeSuccess(bShowDebugSwipeSuccess:bool=true) -> (Output_Get:bool)
   > Set DebugSwipeStartPos(DebugSwipeStartPos:Vector 2D Structure=@n1.StartPos) -> (Output_Get:Vector 2D Structure)
   > Set DebugSwipeEndPos(DebugSwipeEndPos:Vector 2D Structure=@n1.EndPos) -> (Output_Get:Vector 2D Structure)
   > Delay(Duration=1)
   > Set bShowDebugSwipeSuccess(bShowDebugSwipeSuccess=false)
   Event PreConstruct()
   Event Construct()
   Event Tick()
  ]}
 OnPaint {nodes [
   n1 = OnPaint!() -> (then, Context:"Paint Context Structure (by ref)") {is=title K2Node_FunctionEntry}
   > n2 = Branch(Condition=IsPackagedForDistribution!() -> (ReturnValue:bool))
   > Return Node()
   n3 = GetConsoleVariableBoolValue!(execute=@n2.else, VariableName:String=InputZones.ShowDebugInputZones) -> (then, ReturnValue:bool)
   > n4 = Branch(Condition=OR Boolean(A=@n3, B=Get bShowDebugInputZones() -> (bShowDebugInputZones:bool)))
   > n5 = Reverse for Each Loop!(Exec:Exec, Array:Array of Input Zone Structures=Get InputZones() -> (InputZones:Array of Input Zone Structures)) -> (LoopBody:Exec, ArrayIndex:int, ArrayElement:Input Zone Structure, Completed:Exec) {is=title K2Node_MacroInstance}
   n6 = GetConsoleVariableBoolValue(execute=[@n4.else, @n5.Completed], VariableName=InputZones.SwipeOverdriveDebugDisplay)
   > n7 = Branch(Condition=AND Boolean(A=@n6, B=Get bShowDebugSwipeSuccess() -> (bShowDebugSwipeSuccess:bool)))
   n8 = Get DebugSwipeOffsetPos() -> (DebugSwipeOffsetPos:Vector 2D Structure)
   > DrawLine!(
    execute
    Context:"Paint Context Structure (by ref)"=@n1
    PositionA:Vector 2D Structure=vector2d + vector2d(A:Vector 2D Structure=Get DebugSwipeStartPos() -> (DebugSwipeStartPos:Vector 2D Structure), B:Vector 2D Structure=@n8) -> (ReturnValue:Vector 2D Structure) {is=title K2Node_PromotableOperator}
    PositionB:Vector 2D Structure=vector2d + vector2d(A=@n8, B=Get DebugSwipeEndPos() -> (DebugSwipeEndPos:Vector 2D Structure))
    Tint:Linear Color Structure=(R=1,G=0,B=0,A=1)
    bAntiAlias:bool=true
    Thickness:float=2
   ) -> (then)
   > Return Node()
   Return Node(execute=@n7.else)
   n9 = Break Input Zone!(InputZone:"Input Zone Structure (by ref)"=@n5.ArrayElement) -> (
    ZoneWidget:User Widget Object Reference
    ZoneAction:Input Action Object Reference
    ZoneVertices:Array of Vector 2D Structures
    Priority:int
    ZoneID:String
    Activated:bool=~
    DeactivateOnRelease:bool=~
   ) {is=title K2Node_BreakStruct}
   n10 = For Each Loop(Exec=@n5.LoopBody, Array:Array of Vector 2D Structures=@n9.ZoneVertices) -> (Array Element:Vector 2D Structure)
   n11 = "% (Integer)"(A=@n5, B=Length(TargetArray:Array of Linear Color Structures=Get PolygonColors() -> (PolygonColors:Array of Linear Color Structures))) -> (ReturnValue=0)
   n12 = Get PolygonColors()
   n13 = GetLocalSize!(Geometry:"Geometry Structure (by ref)"=GetCachedGeometry!(self:Widget Object Reference) -> (ReturnValue:"Geometry Structure (by ref)")) -> (ReturnValue:Vector 2D Structure)
   n14 = "% (Integer)"(A=int + int(A=@n10.Array Index), B=Length(TargetArray:Array of Vector 2D Structures=@n9.ZoneVertices)) -> (ReturnValue=0)
   > DrawLine(
    Context=@n1
    PositionA=vector2d * vector2d(A:Vector 2D Structure=@n10, B:Vector 2D Structure=@n13) -> (ReturnValue:Vector 2D Structure) {is=title K2Node_PromotableOperator}
    PositionB=vector2d * vector2d(A="Get (a ref)"(Array:Array of Vector 2D Structures=@n9.ZoneVertices, Dimension 1=@n14) -> (Output:"Vector 2D Structure (by ref)"), B=@n13)
    Tint=SelectColor(
     A="Get (a ref)"(Array:Array of Linear Color Structures=@n12, Dimension 1=@n11) -> (Output:"Linear Color Structure (by ref)")
     B=(R=0.1198,G=0.1198,B=0.1198,A=1)
     bPickA=@n9.Activated
    )
    bAntiAlias=false
    Thickness=5
   )
  ]}
]
variables [
 bShowDebugInputZones {type=boolean}
 bShowDebugSwipeSuccess {type=boolean}
 DebugSwipeEndPos {type=Vector2D, default=<0 0>}
 DebugSwipeOffsetPos {type=Vector2D, default=<0 -50>}
 DebugSwipeStartPos {type=Vector2D, default=<0 0>}
 PolygonColors {
  type="array<LinearColor>"
  default [<rgba 0.0491 0.4219 0.0169 1>, <rgba 1 0.0033 0.0146 1>, <rgba 0.9633 1 0 1>, <rgba 0.0004 0.002 0.8333 1>, <rgba 1 0.341 0 1>]
 }
]
widgets {
 info {inheritedWidgetCount=0, parentClass=/Script/PolygonalInputZonesUI.PolygonalInputZoneControllerBase, rootWidgetClass=null, widgetCount=0}
 widgets []
}
